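# ImportFriends is the FortSocialImportPanel widget that prompts players to import platform friends (Facebook, VK, Steam, Xbox, PlayStation, Switch) as Epic friends.
# On activation it reads the social platform, then sets description texts and the visibility of launcher and settings notes.
# Helper functions shrink long header, description and button text via ScaleBox scale, widening text wrap to compensate.
# Intro_V2 plays reversed from its end time, then forward after a short delay.
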WidgetBlueprint /FortniteCoreUI/UI/Friends_UI/ImportingFriends/ImportFriends
animations [
 Intro {
  binding_count=2
  display_label=""
  length_seconds=1.1159
  object=&self:Intro
  sequence {bindings [
    CT_RewardHeader {
     id=#1
     possessed_class=CommonTextBlock
     tracks [{
       type=MovieScene2DTransformTrack
       sections [{
         type=MovieScene2DTransformSection
         end_frame=18
         start_frame=-1
         channels [
          Translation.X {type=MovieSceneScriptingFloatChannel, thinned_from=5, keys [[0,-500],[4,0],[18,0]]}
          Translation.Y {type=MovieSceneScriptingFloatChannel, keys [[0,0],[14,0],[15,22],[18,0]]}
          Angle {type=MovieSceneScriptingFloatChannel}
          Scale.X {type=MovieSceneScriptingFloatChannel, const [1,3]}
          Scale.Y {type=MovieSceneScriptingFloatChannel, keys [[14,1],[15,0.8],[17,1]]}
          Shear.X {type=MovieSceneScriptingFloatChannel, keys [[-0.3177,34.8808],[3,34.8808],[5,-18],[7,15],[9,0]]}
          Shear.Y {type=MovieSceneScriptingFloatChannel, const [0,1]}
         ]
        }]
      }]
    }
    CT_LimitedTimeHeader {
     id=#2
     possessed_class=CommonTextBlock
     tracks [{
       type=MovieScene2DTransformTrack
       sections [{
         type=MovieScene2DTransformSection
         end_frame=22
         start_frame=0
         channels [
          Translation.X {type=MovieSceneScriptingFloatChannel, const [0,2]}
          Translation.Y {type=MovieSceneScriptingFloatChannel, keys [[0,-191],[12,-191],[15,31],[17,-8.9599],[20,10.5658],[22,0]]}
          Angle {type=MovieSceneScriptingFloatChannel}
          Scale.X {type=MovieSceneScriptingFloatChannel, const [1,1]}
          Scale.Y {type=MovieSceneScriptingFloatChannel, keys [[0,2],[14,2],[15,0.75],[17,1]]}
          Shear.X {type=MovieSceneScriptingFloatChannel}
          Shear.Y {type=MovieSceneScriptingFloatChannel}
         ]
        }]
      }]
    }
   ]}
 }
 ArrowPulse {
  binding_count=1
  display_label=""
  length_seconds=1
  object=&self:ArrowPulse
  sequence {bindings [Image_2 {
     id=#3
     possessed_class=Image
     tracks [{
       type=MovieScene2DTransformTrack
       sections [{
         type=MovieScene2DTransformSection
         end_frame=20
         start_frame=0
         channels [
          Translation.X {type=MovieSceneScriptingFloatChannel, keys [[0,0],[10,50],[20,0]]}
          Translation.Y {type=MovieSceneScriptingFloatChannel}
          Angle {type=MovieSceneScriptingFloatChannel}
          Scale.X {type=MovieSceneScriptingFloatChannel, keys [[0,1],[10,0.8509],[20,1]]}
          Scale.Y {type=MovieSceneScriptingFloatChannel}
          Shear.X {type=MovieSceneScriptingFloatChannel}
          Shear.Y {type=MovieSceneScriptingFloatChannel}
         ]
        }]
      }]
    }]}
 }
 Intro_V2 {
  binding_count=9
  display_label=""
  length_seconds=1.45
  object=&self:Intro_V2
  sequence {bindings [
    Button_Add {
     id=#4
     possessed_class=IconTextButton_C
     tracks [{
       type=MovieScene2DTransformTrack
       sections [{
         type=MovieScene2DTransformSection
         end_frame=23
         start_frame=17
         channels [
          Translation.X {type=MovieSceneScriptingFloatChannel}
          Translation.Y {type=MovieSceneScriptingFloatChannel}
          Angle {type=MovieSceneScriptingFloatChannel}
          Scale.X {type=MovieSceneScriptingFloatChannel, keys [[17,0.4],[20,1.1],[22,0.9358],[23,1]]}
          Scale.Y {type=MovieSceneScriptingFloatChannel, keys [[17,0.4],[20,1.1],[22,0.9358],[23,1]]}
          Shear.X {type=MovieSceneScriptingFloatChannel}
          Shear.Y {type=MovieSceneScriptingFloatChannel}
         ]
        }]
      }]
    }
    VBoxAddFriendsConsole {
     id=#5
     possessed_class=VerticalBox
     tracks [{
       type=MovieSceneFloatTrack
       display_name=Render Opacity
       sections [{
         type=MovieSceneFloatSection
         end_frame=23
         start_frame=0
         channels [None {type=MovieSceneScriptingFloatChannel, keys [[0,0],[16,0],[19,1],[23,1]]}]
        }]
      }]
    }
    Button_OptOut {
     id=#6
     possessed_class=IconTextButton_C
     tracks [{
       type=MovieSceneFloatTrack
       display_name=Render Opacity
       sections [{type=MovieSceneFloatSection, end_frame=29, start_frame=0, channels [None {type=MovieSceneScriptingFloatChannel, keys [[0,0],[26,0],[29,1]]}]}]
      }]
    }
    Button_Cancel {
     id=#7
     possessed_class=IconTextButton_C
     tracks [{
       type=MovieSceneFloatTrack
       display_name=Render Opacity
       sections [{
         type=MovieSceneFloatSection
         end_frame=29
         start_frame=0
         channels [None {type=MovieSceneScriptingFloatChannel, thinned_from=4, keys [[0,0],[26,0],[29,1]]}]
        }]
      }]
    }
    CT_LimitedTimeHeader {
     id=#8
     possessed_class=CommonTextBlock
     tracks [{
       type=MovieScene2DTransformTrack
       sections [{
         type=MovieScene2DTransformSection
         end_frame=9
         start_frame=0
         channels [
          Translation.X {type=MovieSceneScriptingFloatChannel, keys [[0,-850],[4,0]]}
          Translation.Y {type=MovieSceneScriptingFloatChannel}
          Angle {type=MovieSceneScriptingFloatChannel}
          Scale.X {type=MovieSceneScriptingFloatChannel}
          Scale.Y {type=MovieSceneScriptingFloatChannel}
          Shear.X {type=MovieSceneScriptingFloatChannel, keys [[0,16],[4,16],[5,-10],[7,0]]}
          Shear.Y {type=MovieSceneScriptingFloatChannel}
         ]
        }]
      }]
    }
    CT_RewardHeader {
     id=#9
     possessed_class=CommonTextBlock
     tracks [{
       type=MovieScene2DTransformTrack
       sections [{
         type=MovieScene2DTransformSection
         end_frame=16
         start_frame=0
         channels [
          Translation.X {type=MovieSceneScriptingFloatChannel, const [0,3]}
          Translation.Y {type=MovieSceneScriptingFloatChannel, keys [[0,1000],[7,1000],[10,0]]}
          Angle {type=MovieSceneScriptingFloatChannel}
          Scale.X {type=MovieSceneScriptingFloatChannel, const [1,2]}
          Scale.Y {type=MovieSceneScriptingFloatChannel, keys [[0,1.5],[9,1.5],[10,0.5765],[12,1.15],[16,1]]}
          Shear.X {type=MovieSceneScriptingFloatChannel}
          Shear.Y {type=MovieSceneScriptingFloatChannel}
         ]
        }]
      }]
    }
    Button_Import {
     id=#10
     possessed_class=IconTextButton_C
     tracks [{
       type=MovieScene2DTransformTrack
       sections [{
         type=MovieScene2DTransformSection
         end_frame=23
         start_frame=17
         channels [
          Translation.X {type=MovieSceneScriptingFloatChannel}
          Translation.Y {type=MovieSceneScriptingFloatChannel}
          Angle {type=MovieSceneScriptingFloatChannel}
          Scale.X {type=MovieSceneScriptingFloatChannel, keys [[17,0.4],[20,1.1],[22,0.9358],[23,1]]}
          Scale.Y {type=MovieSceneScriptingFloatChannel, keys [[17,0.4],[20,1.1],[22,0.9358],[23,1]]}
          Shear.X {type=MovieSceneScriptingFloatChannel}
          Shear.Y {type=MovieSceneScriptingFloatChannel}
         ]
        }]
      }]
    }
    Button_NotNow {
     id=#11
     possessed_class=IconTextButton_C
     tracks [{
       type=MovieSceneFloatTrack
       display_name=Render Opacity
       sections [{
         type=MovieSceneFloatSection
         end_frame=29
         start_frame=0
         channels [None {type=MovieSceneScriptingFloatChannel, thinned_from=4, keys [[0,0],[26,0],[29,1]]}]
        }]
      }]
    }
    VBoxAddFriends {
     id=#12
     display_name=VBoxAddFriendsButton
     possessed_class=VerticalBox
     tracks [{
       type=MovieSceneFloatTrack
       display_name=Render Opacity
       sections [{
         type=MovieSceneFloatSection
         end_frame=23
         start_frame=0
         channels [None {type=MovieSceneScriptingFloatChannel, keys [[0,0],[16,0],[19,1],[23,1]]}]
        }]
      }]
    }
   ]}
 }
]
graphs [
 EventGraph {nodes [
   HandleHeaderText() {is=title K2Node_CustomEvent}
   n1 = Get CT_LimitedTimeHeader() -> (CT_LimitedTimeHeader:Common Text Object Reference)
   > "SetText (Text)"(self=@n1, InText=TextToUpper!(InText:"Text (by ref)"=MakeLiteralText(Value="LOCTEXT(\"add friends\")")) -> (ReturnValue:Text))
   > ScaleTextboxWithWrappingWorkaround!(
    execute
    self
    ContainingScalebox:Scale Box Object Reference=Get ScaleBoxHeader() -> (ScaleBoxHeader:Scale Box Object Reference)
    CharLimitThreshold:int=10
    TargetText:Common Text Object Reference=@n1
    WrapAtHorrizontal:int=Get WrapHeaderAt() -> (WrapHeaderAt:int=0)
   ) -> (then)
   HandleDescriptionText() {is=title K2Node_CustomEvent}
   > n2 = Branch(Condition=Get bHasReward() -> (bHasReward:bool))
   "SetText (Text)"(
    execute=@n2.else
    self=Get CT_AddFriendsDescription() -> (CT_AddFriendsDescription:Common Text Object Reference)
    InText=Select!(
     Facebook:Text=Get FacebookHeader() -> (FacebookHeader:Text)
     VK:Text=Get VKHeader() -> (VKHeader:Text)
     Steam:Text=Get SteamHeader() -> (SteamHeader:Text)
     Xbox:Text=Get XboxHeader() -> (XboxHeader:Text)
     Playstation:Text=Get PlaystationHeader() -> (PlaystationHeader:Text)
     Switch:Text
     None:Text
     Index:ESocialImportPanelPlatform Enum=Get SocialPlatform() -> (SocialPlatform:ESocialImportPanelPlatform Enum=Facebook)
    ) -> (ReturnValue:Text)
   )
   n3 = Get XboxDescriptionText() -> (XboxDescriptionText:Text)
   > "SetText (Text)"(
    self=Get AddFriendsDescription() -> (AddFriendsDescription:Common Text Object Reference)
    InText=Select(
     Facebook=@n3
     VK=@n3
     Steam=@n3
     Xbox=@n3
     Playstation=Get PSNDescriptionText() -> (PSNDescriptionText:Text)
     Switch=Get SwitchDescriptionText() -> (SwitchDescriptionText:Text)
     Index=Get SocialPlatform() -> (SocialPlatform=Facebook)
    )
   )
   > ScaleTextboxWithWrappingWorkaround(
    ContainingScalebox=Get ScaleBoxDescription() -> (ScaleBoxDescription:Scale Box Object Reference)
    CharLimitThreshold=350
    TargetText=Get CT_AddFriendsDescription()
    WrapAtHorrizontal=Get WrapHeaderAt() -> (WrapHeaderAt=0)
   )
   Event PreConstruct()
   > HandleHeaderText() {is=title K2Node_CallFunction}
   > HandleDescriptionText() {is=title K2Node_CallFunction}
   > n4 = Cast To IconTextButton!(execute, Object=Get Button_Import() -> (Button_Import:Common Button Legacy Object Reference)) -> (then, CastFailed, AsIcon Text Button:Icon Text Button Object Reference) {is=title K2Node_DynamicCast}
   > ScaleTextButton!(
    execute
    self
    CharThreshold:int=30
    ContainingScaleBox:Scale Box Object Reference=Get SBox_AddFriendButtonScaleBox() -> (SBox_AddFriendButtonScaleBox:Scale Box Object Reference)
    Button:Icon Text Button Object Reference=@n4
   ) -> (then)
   Event On Activated()
   > Set SocialPlatform(SocialPlatform:ESocialImportPanelPlatform Enum=GetSocialPlatform!(self:Fort Social Import Panel Object Reference) -> (OutPlatform:ESocialImportPanelPlatform Enum=Facebook, ReturnValue:bool)) -> (Output_Get:ESocialImportPanelPlatform Enum=Facebook)
   > n5 = Cast To IconTextButton(Object=Get Button_Import())
   > n6 = ScaleTextButton(CharThreshold=30, ContainingScaleBox=Get SBox_AddFriendButtonScaleBox(), Button=@n5)
   HandleHeaderText(execute=[@n6.then, @n5.CastFailed])
   > HandleDescriptionText()
   > SetVisibility(
    self=Get CT_TakenToLauncher() -> (CT_TakenToLauncher:Common Text Object Reference)
    InVisibility=Select(
     Facebook:ESlateVisibility Enum=HitTestInvisible
     VK:ESlateVisibility Enum=HitTestInvisible
     Steam:ESlateVisibility Enum=HitTestInvisible
     Xbox:ESlateVisibility Enum=Collapsed
     Playstation:ESlateVisibility Enum=Collapsed
     Switch:ESlateVisibility Enum=Visible
     None:ESlateVisibility Enum=Visible
     Index=Get SocialPlatform() -> (SocialPlatform=Facebook)
    ) -> (ReturnValue:ESlateVisibility Enum)
   )
   > SetVisibility(
    self=Get CT_ChangeInSettingsText() -> (CT_ChangeInSettingsText:Common Text Object Reference)
    InVisibility=Select(
     Facebook=Collapsed
     VK=Collapsed
     Steam=Collapsed
     Xbox=Visible
     Playstation=Visible
     Switch=Visible
     None=Visible
     Index=Get SocialPlatform() -> (SocialPlatform=Facebook)
    )
   )
   n7 = Get Intro_V2() -> (Intro_V2:Widget Animation Object Reference)
   > PlayAnimation(InAnimation=@n7, StartAtTime=GetEndTime(self=@n7), PlayMode=Reverse)
   > Delay(Duration=Get HeartbeatDelayIntroAnimation() -> (HeartbeatDelayIntroAnimation:double))
   > PlayAnimation(InAnimation=Get Intro_V2())
   Event OnIncentivizedSet!() -> (OutputDelegate, then, bIncentivized:bool) {is=title K2Node_Event}
   > Set bHasReward(bHasReward:bool=false) -> (Output_Get:bool)
   n8 = Branch(Condition=Get bHasReward())
   > "SetText (Text)"(self=Get CT_LimitedTimeHeader(), InText=TextToUpper(InText=MakeLiteralText(Value="LOCTEXT(\"limited time\")")))
   n9 = Select(
    Facebook:Text=Get FacebookHeader()
    VK:Text=Get VKHeader()
    Steam:Text=Get SteamHeader()
    Xbox:Text=Get XboxHeader()
    Playstation:Text=Get PlaystationHeader()
    Switch:Text
    None:Text
    Index=Get SocialPlatform() -> (SocialPlatform=Facebook)
   ) -> (ReturnValue:Text)
   "SetText (Text)"(execute=@n8.else, self=Get CT_LimitedTimeHeader(), InText=TextToUpper(InText=MakeLiteralText(Value=@n9)))
   n10 = Get PreviewItemDef() -> (PreviewItemDef:Fort Item Definition Object Reference)
   "SetText (Text)"(
    self=Get CT_AddFriendsDescription()
    InText=Format Text!(
     Format="LOCTEXT(\"Add your {x} friends as Epic friends and get the exclusive {y} outfit today!\")"
     x:Wildcard
     y:Text=Get ItemName!(self:Item Definition Base Object Reference=@n10) -> (ItemName:Text)
    ) -> (Result)
   )
   "On Button Clicked (Button_Cancel)"!() -> (OutputDelegate, then, Button:Common Button Legacy Object Reference) {is=title K2Node_ComponentBoundEvent}
   Select!(
    Athena:Wildcard
    SaveTheWorld:Wildcard
    Index:ESocialImportPanelType Enum=Event OnPanelTypeSet!() -> (OutputDelegate, then, NewType:ESocialImportPanelType Enum=Athena) {is=title K2Node_Event}.NewType
   ) -> (ReturnValue:Wildcard)
   Get Intro()
  ]}
 ScaleTextButton {nodes [
   ScaleTextButton!() -> (then, CharThreshold:int, ContainingScaleBox:Scale Box Object Reference, Button:Icon Text Button Object Reference) {is=title K2Node_FunctionEntry}
   n1 = Get Button() -> (Button:Icon Text Button Object Reference)
   n2 = Len(S="To String (Text)"(InText=Get ButtonText!(self:Icon Text Button Object Reference=@n1) -> (ButtonText:Text)))
   n3 = Get CharThreshold() -> (CharThreshold:int=0)
   n4 = "Max (Float)"(A=float / float(A="To Float (Integer)"(InInt=@n2), B="To Float (Integer)"(InInt=@n3)) {is=title K2Node_CallFunction}, B=1)
   > SetUserSpecifiedScale!(
    execute
    self:Scale Box Object Reference=Get ContainingScaleBox() -> (ContainingScaleBox:Scale Box Object Reference)
    InUserSpecifiedScale:float="Clamp (Float)"(Value=float / float(A=1, B=@n4), Min=0.3, Max=1)
   ) -> (then)
  ]}
 ScaleTextboxWithWrappingWorkaround {nodes [
   ScaleTextboxWithWrappingWorkaround!() -> (then, ContainingScalebox:Scale Box Object Reference, CharLimitThreshold:int, TargetText:Common Text Object Reference, WrapAtHorrizontal:int) {is=title K2Node_FunctionEntry}
   n1 = "To String (Text)"(InText="GetText (Text)"!(self:Text Object Reference=Get TargetText() -> (TargetText:Common Text Object Reference)) -> (ReturnValue:Text))
   n2 = float / float(A="To Float (Integer)"(InInt=Len(S=@n1)), B="To Float (Integer)"(InInt=Get CharLimitThreshold() -> (CharLimitThreshold:int=0)))
   n3 = "Clamp (Float)"(Value=float / float(A=1, B="Max (Float)"(A=@n2, B=1)), Min=0.3, Max=0.9091)
   > SetUserSpecifiedScale(self=Get ContainingScalebox() -> (ContainingScalebox:Scale Box Object Reference), InUserSpecifiedScale=@n3)
   n4 = "To Float (Integer)"(InInt=Get WrapAtHorrizontal() -> (WrapAtHorrizontal:int=0))
   > SetWrapTextWidth!(execute, self:Common Text Object Reference=Get TargetText(), InWrapTextAt:int=Truncate(A=float / float(A=@n4, B=@n3))) -> (then)
  ]}
]
variables [
 bHasReward {type=boolean}
 FacebookHeader {type=string, default="Add your Facebook friends as Epic friends to fill out your squad!"}
 HeartbeatDelayIntroAnimation {type=number, default=0.15}
 PlaystationHeader {type=string, default="Add your PlayStation™ Network friends to fill out your squad!"}
 PreviewItemDef {type=FortItemDefinition, default=/Game/Athena/Items/Cosmetics/Characters/CID_044_Athena_Commando_F_SciPop}
 PSNDescriptionText {
  type=string
  default="Your PSN friends can automatically be added as Epic friends to fill out your squad. Fortnite is more fun with friends!"
 }
 SocialPlatform {type=ESocialImportPanelPlatform, default=Facebook}
 SteamHeader {type=string, default="Add your Steam friends as Epic friends to fill out your squad!"}
 SwitchDescriptionText {
  type=string
  default="Your Nintendo friends can automatically be added as Epic friends to fill out your squad. Fortnite is more fun with friends!"
 }
 VKHeader {type=string, default="Add your VK friends as Epic friends to fill out your squad!"}
 WrapHeaderAt {type=integer, default=900}
 XboxDescriptionText {
  type=string
  default="Your Xbox Live friends can automatically be added as Epic friends to fill out your squad. Fortnite is more fun with friends!"
 }
 XboxHeader {type=string, default="Add your Xbox Live friends as Epic friends to fill out your squad!"}
]
widgets {
 info {inheritedWidgetCount=7, parentClass=/Script/FortniteUI.FortSocialImportPanel, rootWidgetClass=&Overlay, widgetCount=28}
 widgets [
  Overlay_2 {
   widgetClassPath=&Overlay
   widget=Overlay_2
   properties {navigation=&self:WidgetTree.Overlay_2.WidgetNavigation_0}
  }
  Border_OuterBorder {
   widgetClassPath=&Border
   bInherited=true
   bIsVariable=true
   parent=Overlay_2
   widget=Border_OuterBorder
   properties {
    navigation=&self:WidgetTree.Border_OuterBorder.WidgetNavigation_0
    background {
     bIsDynamicallyLoaded=false
     drawAs=Image
     imageType=NoImage
     mirroring=NoMirror
     resourceName=None
     resourceObject=/FortniteCoreUI/UI/HomeBase/Materials/ImportFriendGradient
     tiling=NoTile
     imageSize=<32 32>
     margin=<ltrb 0 0 0 0>
     outlineSettings {
      bUseBrushTransparency=false
      roundingType=HalfHeightRadius
      width=0
      color {colorUseRule=UseColor_Specified, specifiedColor=<rgba 0 0 0 0>}
      cornerRadii=<0 0 0 0>
     }
     tintColor {colorUseRule=UseColor_Specified, specifiedColor=<rgba 1 1 1 1>}
     uVRegion {bIsValid=false, max=<0 0>, min=<0 0>}
    }
    padding=<ltrb 0 0 0 0>
   }
   slot {class=OverlaySlot, properties {horizontalAlignment=HAlign_Fill, verticalAlignment=VAlign_Fill}}
  }
  SboxBackground {
   widgetClassPath=&ScaleBox
   parent=Overlay_2
   widget=SboxBackground
   properties {stretch=ScaleToFitY, stretchDirection=DownOnly}
   slot {class=OverlaySlot, properties {horizontalAlignment=HAlign_Fill, verticalAlignment=VAlign_Fill}}
  }
  Image_1 {
   widgetClassPath=&Image
   bIsVariable=true
   parent=SboxBackground
   widget=Image_1
   properties {
    brush {
     bIsDynamicallyLoaded=false
     drawAs=Image
     imageType=NoImage
     mirroring=NoMirror
     resourceName=None
     resourceObject=/FortniteCoreUI/UI/Login/Art/T_LoginSplashArt_1024
     tiling=NoTile
     imageSize=<1024 1024>
     margin=<ltrb 0 0 0 0>
     outlineSettings {
      bUseBrushTransparency=false
      roundingType=HalfHeightRadius
      width=0
      color {colorUseRule=UseColor_Specified, specifiedColor=<rgba 0 0 0 0>}
      cornerRadii=<0 0 0 0>
     }
     tintColor {colorUseRule=UseColor_Specified, specifiedColor=<rgba 1 1 1 1>}
     uVRegion {bIsValid=false, max=<0 0>, min=<0 0>}
    }
    renderTransform {angle=0, scale=<1.15 1.15>, shear=<0 0>, translation=<0 0>}
   }
   slot {class=ScaleBoxSlot, properties {horizontalAlignment=HAlign_Right, verticalAlignment=VAlign_Bottom}}
  }
  SafeZone_0 {
   widgetClassPath=&SafeZone
   bIsVariable=true
   parent=Overlay_2
   widget=SafeZone_0
   properties {padBottom=false, padTop=false}
   slot {class=OverlaySlot, properties {horizontalAlignment=HAlign_Fill, verticalAlignment=VAlign_Fill, padding=<ltrb 4 2 4 2>}}
  }
  MainContainer {
   widgetClassPath=&Overlay
   parent=SafeZone_0
   widget=MainContainer
   slot {class=SafeZoneSlot, properties {padding=<ltrb 80 40 80 40>}}
  }
  RightContainer {
   widgetClassPath=&Overlay
   parent=MainContainer
   widget=RightContainer
   slot {class=OverlaySlot, properties {horizontalAlignment=HAlign_Right, verticalAlignment=VAlign_Fill}}
  }
  VerticalBox_1 {
   widgetClassPath=&VerticalBox
   parent=RightContainer
   widget=VerticalBox_1
   slot {class=OverlaySlot, properties {horizontalAlignment=HAlign_Right, verticalAlignment=VAlign_Bottom, padding=<ltrb 0 0 40 0>}}
  }
  ItemInfoWidget {
   widgetClassPath=/Game/Athena/UI/Season/ItemInfoWidget
   bIsVariable=true
   parent=VerticalBox_1
   widget=ItemInfoWidget
   slot {class=VerticalBoxSlot}
  }
  Button_OptOut {
   widgetClassPath=&IconTextButton
   bInherited=true
   bIsVariable=true
   parent=VerticalBox_1
   widget=Button_OptOut
   properties {
    buttonText=Never Show Again
    minHeight=85
    onButtonBaseClicked=(&self:WidgetTree.Button_OptOut.HandleOnButtonClicked)
    onButtonBaseDoubleClicked=(&self:WidgetTree.Button_OptOut.HandleOnButtonDoubleClicked)
    onButtonBaseHovered=(&self:WidgetTree.Button_OptOut.HandleOnButtonHovered)
    onButtonBaseUnhovered=(&self:WidgetTree.Button_OptOut.HandleOnButtonUnhovered)
    onSelectedChangedBase=(&self:WidgetTree.Button_OptOut.HandleOnSelectedChanged)
    style=/FortUILibrary/Foundation/Buttons/ButtonStyles/Skew/ButtonStyle-Skew_LessDesirable.ButtonStyle-Skew_LessDesirable_C
   }
   slot {class=VerticalBoxSlot, properties {horizontalAlignment=HAlign_Right, padding=<ltrb 0 20 0 0>}}
  }
  VerticalBox_0 {
   widgetClassPath=&VerticalBox
   parent=MainContainer
   widget=VerticalBox_0
   slot {class=OverlaySlot, properties {horizontalAlignment=HAlign_Fill, verticalAlignment=VAlign_Fill}}
  }
  ScaleBoxHeader {
   widgetClassPath=&ScaleBox
   bIsVariable=true
   parent=VerticalBox_0
   widget=ScaleBoxHeader
   properties {stretch=UserSpecified, userSpecifiedScale=0.9267}
   slot {
    class=VerticalBoxSlot
    properties {horizontalAlignment=HAlign_Left, verticalAlignment=VAlign_Top, padding=<ltrb 0 -25 0 0>, size {sizeRule=Fill, value=1}}
   }
  }
  CT_LimitedTimeHeader {
   widgetClassPath=&CommonTextBlock
   bIsVariable=true
   parent=ScaleBoxHeader
   widget=CT_LimitedTimeHeader
   properties {
    autoWrapText=true
    lineHeightPercentage=0.75
    wrapTextAt=1300
    font {
     bForceMonospaced=false
     bMaterialIsStencil=false
     fontMaterial=null
     fontObject=&Burbank
     letterSpacing=0
     monospacedWidth=1
     size=175
     skewAmount=0
     typefaceFontName=Burbank-Black
     outlineSettings {
      bApplyOutlineToDropShadows=true
      bMiteredCorners=false
      bSeparateFillAlpha=false
      outlineMaterial=null
      outlineSize=3
      outlineColor=<rgba 0 0 0 0.1>
     }
    }
    renderTransformPivot=<0.5 1>
    shadowOffset=<0 0>
   }
   slot {class=ScaleBoxSlot}
  }
  Switcher_AddFriends {
   widgetClassPath=&WidgetSwitcher
   bInherited=true
   bIsVariable=true
   parent=VerticalBox_0
   widget=Switcher_AddFriends
   slot {class=VerticalBoxSlot, properties {horizontalAlignment=HAlign_Left, size {sizeRule=Fill, value=1}}}
  }
  VBoxAddFriends {widgetClassPath=&VerticalBox, parent=Switcher_AddFriends, widget=VBoxAddFriends, slot {class=WidgetSwitcherSlot}}
  ScaleBoxDescription {
   widgetClassPath=&ScaleBox
   bIsVariable=true
   parent=VBoxAddFriends
   widget=ScaleBoxDescription
   properties {stretch=UserSpecified}
   slot {class=VerticalBoxSlot}
  }
  CT_AddFriendsDescription {
   widgetClassPath=&CommonTextBlock
   bIsVariable=true
   parent=ScaleBoxDescription
   widget=CT_AddFriendsDescription
   properties {
    text="Add your Playstation friends as Epic friends and get the exclusive {ITEM_NAME} outfit today!"
    wrapTextAt=825.5289
    font {
     bForceMonospaced=false
     bMaterialIsStencil=false
     fontMaterial=null
     fontObject=&BurbankSmall
     letterSpacing=0
     monospacedWidth=1
     size=32
     skewAmount=0
     typefaceFontName=Bold
     outlineSettings {
      bApplyOutlineToDropShadows=false
      bMiteredCorners=false
      bSeparateFillAlpha=false
      outlineMaterial=null
      outlineSize=0
      outlineColor=<rgba 0 0 0 1>
     }
    }
    shadowOffset=<0 0>
   }
   slot {class=ScaleBoxSlot, properties {horizontalAlignment=HAlign_Left}}
  }
  SBox_AddFriendButtonScaleBox {
   widgetClassPath=&ScaleBox
   bIsVariable=true
   parent=VBoxAddFriends
   widget=SBox_AddFriendButtonScaleBox
   properties {stretch=UserSpecified}
   slot {class=VerticalBoxSlot, properties {horizontalAlignment=HAlign_Left, padding=<ltrb 0 16 0 0>}}
  }
  Button_Import {
   widgetClassPath=&IconTextButton
   bInherited=true
   bIsVariable=true
   parent=SBox_AddFriendButtonScaleBox
   widget=Button_Import
   properties {
    buttonText=Add Friends Now
    contentAlignment=HAlign_Center
    minHeight=112
    minWidth=512
    onButtonBaseClicked=(&self:WidgetTree.Button_Import.HandleOnButtonClicked)
    onButtonBaseDoubleClicked=(&self:WidgetTree.Button_Import.HandleOnButtonDoubleClicked)
    onButtonBaseHovered=(&self:WidgetTree.Button_Import.HandleOnButtonHovered)
    onButtonBaseUnhovered=(&self:WidgetTree.Button_Import.HandleOnButtonUnhovered)
    onSelectedChangedBase=(&self:WidgetTree.Button_Import.HandleOnSelectedChanged)
    style=/FortUILibrary/Foundation/Buttons/ButtonStyles/Skew/ButtonStyle-Skew_Desirable.ButtonStyle-Skew_Desirable_C
   }
   slot {class=ScaleBoxSlot}
  }
  CT_TakenToLauncher {
   widgetClassPath=&CommonTextBlock
   bIsVariable=true
   parent=VBoxAddFriends
   widget=CT_TakenToLauncher
   properties {
    text=Note - You'll be taken to the Epic Games Launcher to finish this process
    font {
     bForceMonospaced=false
     bMaterialIsStencil=false
     fontMaterial=null
     fontObject=&BurbankSmall
     letterSpacing=0
     monospacedWidth=1
     size=16
     skewAmount=0
     typefaceFontName=Bold
     outlineSettings {
      bApplyOutlineToDropShadows=false
      bMiteredCorners=false
      bSeparateFillAlpha=false
      outlineMaterial=null
      outlineSize=0
      outlineColor=<rgba 0 0 0 1>
     }
    }
    shadowOffset=<0 0>
   }
   slot {class=VerticalBoxSlot, properties {padding=<ltrb 4 16 0 0>}}
  }
  CT_ChangeInSettingsText {
   widgetClassPath=&CommonTextBlock
   bIsVariable=true
   parent=VBoxAddFriends
   widget=CT_ChangeInSettingsText
   properties {
    style=/FortUILibrary/Styles/Typography/Foundation/TextStyle-Base-XS.TextStyle-Base-XS_C
    text=NOTE - You can change your preference at any time in Settings.
    font {
     bForceMonospaced=false
     bMaterialIsStencil=false
     fontMaterial=null
     fontObject=&NotoSans
     letterSpacing=0
     monospacedWidth=1
     size=16
     skewAmount=0
     typefaceFontName=NotoSans-Regular
     outlineSettings {
      bApplyOutlineToDropShadows=false
      bMiteredCorners=false
      bSeparateFillAlpha=false
      outlineMaterial=null
      outlineSize=0
      outlineColor=<rgba 0 0 0 1>
     }
    }
    shadowOffset=<0 0>
   }
   slot {class=VerticalBoxSlot, properties {padding=<ltrb 4 0 0 0>}}
  }
  VBoxAddFriendsConsole {
   widgetClassPath=&VerticalBox
   parent=Switcher_AddFriends
   widget=VBoxAddFriendsConsole
   slot {class=WidgetSwitcherSlot, properties {padding=<ltrb 0 -20 0 0>}}
  }
  AddFriendsDescription {
   widgetClassPath=&CommonTextBlock
   bIsVariable=true
   parent=VBoxAddFriendsConsole
   widget=AddFriendsDescription
   properties {
    text="Your Xbox Live friends can automatically be added as Epic Friends to fill out your squad and provide improved social features!\r\n\r\nAdd your Xbox Live friends now?"
    wrapTextAt=768
    font {
     bForceMonospaced=false
     bMaterialIsStencil=false
     fontMaterial=null
     fontObject=&BurbankSmall
     letterSpacing=0
     monospacedWidth=1
     size=24
     skewAmount=0
     typefaceFontName=Bold
     outlineSettings {
      bApplyOutlineToDropShadows=false
      bMiteredCorners=false
      bSeparateFillAlpha=false
      outlineMaterial=null
      outlineSize=0
      outlineColor=<rgba 0 0 0 1>
     }
    }
    shadowOffset=<0 0>
   }
   slot {class=VerticalBoxSlot}
  }
  VerticalBox_4 {
   widgetClassPath=&VerticalBox
   parent=VBoxAddFriendsConsole
   widget=VerticalBox_4
   slot {class=VerticalBoxSlot, properties {horizontalAlignment=HAlign_Left}}
  }
  Button_Add {
   widgetClassPath=&IconTextButton
   bInherited=true
   bIsVariable=true
   parent=VerticalBox_4
   widget=Button_Add
   properties {
    buttonText=ADD
    contentAlignment=HAlign_Center
    minHeight=112
    minWidth=512
    onButtonBaseClicked=(&self:WidgetTree.Button_Add.HandleOnButtonClicked)
    onButtonBaseDoubleClicked=(&self:WidgetTree.Button_Add.HandleOnButtonDoubleClicked)
    onButtonBaseHovered=(&self:WidgetTree.Button_Add.HandleOnButtonHovered)
    onButtonBaseUnhovered=(&self:WidgetTree.Button_Add.HandleOnButtonUnhovered)
    onSelectedChangedBase=(&self:WidgetTree.Button_Add.HandleOnSelectedChanged)
    style=&ButtonStyle-Skew_Desirable_C
   }
   slot {class=VerticalBoxSlot, properties {padding=<ltrb 0 16 0 0>}}
  }
  Button_NotNow {
   widgetClassPath=&IconTextButton
   bInherited=true
   bIsVariable=true
   parent=VerticalBox_4
   widget=Button_NotNow
   properties {
    buttonText=NOT NOW
    contentAlignment=HAlign_Center
    minHeight=80
    minWidth=280
    onButtonBaseClicked=(&self:WidgetTree.Button_NotNow.HandleOnButtonClicked)
    onButtonBaseDoubleClicked=(&self:WidgetTree.Button_NotNow.HandleOnButtonDoubleClicked)
    onButtonBaseHovered=(&self:WidgetTree.Button_NotNow.HandleOnButtonHovered)
    onButtonBaseUnhovered=(&self:WidgetTree.Button_NotNow.HandleOnButtonUnhovered)
    onSelectedChangedBase=(&self:WidgetTree.Button_NotNow.HandleOnSelectedChanged)
    style=/FortUILibrary/Foundation/Buttons/ButtonStyles/Skew/ButtonStyle-Skew_LessDesirable_AddFriend.ButtonStyle-Skew_LessDesirable_AddFriend_C
   }
   slot {class=VerticalBoxSlot}
  }
  CommonTextBlock_0 {
   widgetClassPath=&CommonTextBlock
   bIsVariable=true
   parent=VBoxAddFriendsConsole
   widget=CommonTextBlock_0
   properties {
    text=NOTE - You can change your preference at any time in Settings.
    wrapTextAt=768
    font {
     bForceMonospaced=false
     bMaterialIsStencil=false
     fontMaterial=null
     fontObject=&BurbankSmall
     letterSpacing=0
     monospacedWidth=1
     size=16
     skewAmount=0
     typefaceFontName=Bold
     outlineSettings {
      bApplyOutlineToDropShadows=false
      bMiteredCorners=false
      bSeparateFillAlpha=false
      outlineMaterial=null
      outlineSize=0
      outlineColor=<rgba 0 0 0 1>
     }
    }
    shadowOffset=<0 0>
   }
   slot {class=VerticalBoxSlot, properties {padding=<ltrb 4 24 0 0>}}
  }
  Button_Cancel {
   widgetClassPath=&IconTextButton
   bInherited=true
   bIsVariable=true
   parent=MainContainer
   widget=Button_Cancel
   properties {
    buttonText=Remind Me Later
    contentAlignment=HAlign_Center
    minHeight=85
    minWidth=300
    onButtonBaseClicked=(&self:WidgetTree.Button_Cancel.HandleOnButtonClicked)
    onButtonBaseDoubleClicked=(&self:WidgetTree.Button_Cancel.HandleOnButtonDoubleClicked)
    onButtonBaseHovered=(&self:WidgetTree.Button_Cancel.HandleOnButtonHovered)
    onButtonBaseUnhovered=(&self:WidgetTree.Button_Cancel.HandleOnButtonUnhovered)
    onSelectedChangedBase=(&self:WidgetTree.Button_Cancel.HandleOnSelectedChanged)
    style=&ButtonStyle-Skew_LessDesirable_C
   }
   slot {class=OverlaySlot, properties {verticalAlignment=VAlign_Bottom}}
  }
 ]
}
